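# Translucent, volumetric per-vertex lit material for falling-leaf sprites in Cascade and Niagara particle effects.
# A SubUV leaf texture is multiplied by particle color, with RGB passed through ConvertFromDiffSpec and alpha forming the opacity mask.
# An optional switch, off by default, scales that alpha by an inverted synced distance fade.
# The Use3ChannelColor switch has identical RGB inputs on both branches, so it changes nothing.
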Material /Game/Effects/Fort_Effects/World/Enviroment/Falling_Leaf/M_Leaf_SubUV
graph {
 properties {
  bUsedWithNiagaraSprites=true
  bUsedWithParticleSprites=true
  blendMode=BLEND_Translucent
  pixelDepthOffsetMode=PDOM_Legacy
  translucencyLightingMode=TLM_VolumetricPerVertexNonDirectional
  usedShadingModels=Default Lit
 }
 nodes [
  n1 = TextureSampleParameterSubUV!(UVs, Apply View MipBias) {parameterName=TextureInput, texture=/Game/Effects/Fort_Effects/Textures/Leaf/T_Leaf_SubUV}
  n2 = ParticleColor()
  n3 = Multiply(A=@n1.A, B=@n2.A)
  LinearInterpolate(
   A=ScalarParameter() {defaultValue=1, group=Textures, parameterName=Opacity}
   B=@n3
   Alpha=ScalarParameter() {group=Textures, parameterName=UseOpacityMaskTexture?}
  )
  MaterialFunctionCall(DiffuseColor=Multiply(A=StaticSwitchParameter(True=@n1.RGB, False=@n1.RGB) {defaultValue=true, parameterName=Use3ChannelColor?}, B=@n2.RGB)) {materialFunction=&ConvertFromDiffSpec}
  n4 = ScalarParameter() {defaultValue=500, parameterName="Fade_Start_Distance (S)"}
  n5 = ScalarParameter() {defaultValue=100, parameterName="Fade_End_Distance (S)"}
  StaticSwitchParameter(
   True=Multiply(
    A=MaterialFunctionCall(Fade_Start_Distance=@n4, Fade_End_Distance=@n5) {materialFunction=&MF_Distance_SyncedFade_FX}.Result Inverted
    B=@n3
   )
   False=@n3
  ) {parameterName=UseDistanceFade?}
 ]
}
